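# Niagara environment system for desert-biome blowing sand and tumbleweeds, with GPU emitters driven by wind gusts.
# Sand sprites spawn on sampled landscape, filtered by GBuffer color and surface normal, with time-of-day tinting and hard-coded kill boxes excluding map areas.
# Tumbleweed meshes spawn on the landscape, then roll under wind, gravity and collision.
# Player position drives size and alpha fades; a third emitter, Dirt_Blowing, is disabled.
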NiagaraSystem /Game/Effects/Fort_Effects/Effects/Environment/Hera/DesertSquidWharf/NS_Env_DesertBiome_BlowingSand
dependencies {
 usedDataInterfaces [
  &NiagaraDataInterfaceCurve
  &NiagaraDataInterfaceCollisionQuery
  &NiagaraDataInterfaceVectorCurve
  &NiagaraDataInterfaceSpriteRendererInfo
  &NiagaraDataInterfaceArrayColor
  /Script/Niagara.NiagaraDataInterfaceAsyncGpuTrace
  &NiagaraDataInterfaceDebugDraw
  &NiagaraDataInterfaceGBuffer
  &NiagaraDataInterfaceColorCurve
 ]
 usedDynamicInputs [&MakeFloatFromVector]
 usedModules [
  &SystemState
  &self:SystemSpawnScript.SystemScriptSource.SystemScriptGraph.NiagaraNodeAssignment_0.SetVariables_#1
 ]
}
emitters {
 Dirt_Blowing {
  bEnabled=false
  simTarget=GPUComputeSim
  emitterUpdateScript [
   &EmitterState()
   &SpawnRate(SpawnRate=&RandomRangeFloat(Minimum=5, Maximum=9), Use Spawn Probability=false)
  ]
  particleSpawnScript [
   &InitializeParticle(
    Color=&FNTimeOfDayColor(
     Night Color=<rgba 0.314 0.1418 0.1329 1>
     Morning Color=<rgba 0.5506 0.1778 0.1574 1>
     Day Color=<rgba 0.5506 0.1778 0.1574 1>
     Evening Color=<rgba 0.314 0.1418 0.1329 1>
    )
    Lifetime Mode=::NewEnumerator1
    Sprite Rotation Mode=::NewEnumerator2
    Sprite Size Mode=::NewEnumerator3
    Sprite UV Mode=::NewEnumerator1
    Uniform Sprite Size=&Multiply_Float(A=&RandomRangeFloat(Minimum=500, Maximum=1000), B=0.5)
    UsePositionOffset=false
    Write Material Random=true
   )
   &ShapeLocation(MaskSpawnGroup=false, Override Seed=false, Sphere Radius=1500, UseOwnerScale=true)
   /Niagara/Modules/Landscape/SampleLandscape(Height Offset=10)
   &AddVelocity(Rotation Coordinate Space=::World, Velocity=<0 0 0>, Velocity Speed Scale=30)
   &DynamicMaterialParameters(
    Index 0 Param 1=&RandomRangeFloat(Minimum=50, Maximum=100)
    Index 0 Param 2=5
    Index 0 Param 3=15
    Index 0 Param 4=&RandomRangeFloat(Minimum=0, Maximum=0)
    Index 1 Param 1=50
    Index 2 Param 1=&RandomRangeFloat(Minimum=-0.1, Maximum=-0.1)
    Index 2 Param 2=&RandomRangeFloat(Minimum=0.1, Maximum=0)
    Index 2 Param 3=&RandomRangeFloat(Minimum=0, Maximum=0)
    Index 2 Param 4=&RandomRangeFloat(Minimum=0.2, Maximum=0.3)
    Module.Index0Param0Write=true
    Module.Index0Param1Write=true
    Module.Index0Param2Write=true
    Module.Index0Param3Write=true
    Module.Index1Param0Write=true
    Module.Index1Param1Write=true
    Module.Index1Param2Write=true
    Module.Index1Param3Write=true
    Module.Index2Param0Write=true
    Module.Index2Param1Write=true
    Module.Index2Param2Write=true
    Module.Index2Param3Write=true
    Write Parameter Index 1=true
    Write Parameter Index 2=true
   )
   &KillParticlesInVolume(
    Sphere Radius=250
    Volume Origin:Vector3f=&MakeVectorFromLinearColorRGB(Color=$NPC.FortniteNPC.PlayerPosition)
   ) {enabled=false}
  ]
  particleUpdateScript [
   &ParticleState()
   &ScaleSpriteSize(
    Uniform Curve Scale=1.1
    Uniform Curve Sprite Scale=DataInterface() {propertyValues {Curve {keys [{interpMode=RCIM_Cubic, leaveTangent=3.9356, tangentMode=RCTM_Break, time=0, value=0}, <at 1 1 RCIM_Cubic>]}}}
   )
   &SpriteRotationRate(Rotation Rate=-60)
   &ScaleSpriteSize(
    Uniform Curve Index=&self:ScratchDynamicInput_04(
     End Distance:NiagaraFloat=7500
     Max Output:NiagaraFloat=1
     MCP Player Location:LinearColor=$NPC.FortniteNPC.PlayerPosition
     Minimum Output:NiagaraFloat=0
     Start Distance:NiagaraFloat=0
    )
    Uniform Curve Sprite Scale=DataInterface() {propertyValues {Curve {
       defaultValue=0
       keys [{interpMode=RCIM_Cubic, tangentMode=RCTM_User, time=0, value=0}, {interpMode=RCIM_Cubic, tangentMode=RCTM_User, time=0.1, value=1}]
      }}}
   ) {enabled=false, module=ScaleSpriteSize003}
   &WindForce(
    Offset By Time:Vector3f=<0 0 15>
    Scale=0
    Wind Speed=<750 0 0>
    Wind Speed Scale=&Lerp_Float(A=0.25, B=0.5, Alpha=$NPC.FortniteNPC.WindGustsSystem_Active)
   )
   &Drag(Drag=4, Use Linear Drag=true, Use Rotational Drag=true)
   &GravityForce(Gravity=&VectorFromCurve(
     VectorCurve=DataInterface() {propertyValues {XCurve {keys [<at 0.75 1>,<at 1 0>]}, YCurve {keys [<at 0.75 1>,<at 1 0>]}, ZCurve {keys [<at 0.75 1>,<at 1 0>]}}}
     Scale Curve=<1 1 0>
    )) {enabled=false}
   &SolveForcesAndVelocity(Clamp Velocity=false, Limit Acceleration=false) {module=SolveForcesAndVelocity001}
   &SolveForcesAndVelocity(Clamp Velocity=false, Limit Acceleration=false)
   &ScaleSpriteSizeBySpeed(Max Scale Factor=<2 12>, Min Scale Factor=<0.05 0.05>, Velocity Threshold=2000)
   &RegionMask(
    Invert Region=true
    Region Dimensions:Vector3f=<1 1 1>
    Region Falloff=1
    Region Type:ENiagaraDebugDrawShape=::NewEnumerator1
    UseOffset=true
    UseRotation=false
   )
   &ScaleColor(
    Scale Alpha=&Multiply_Float(
     A=&FloatFromCurve(
      FloatCurve=DataInterface() {propertyValues {Curve {keys [
          {interpMode=RCIM_Cubic, tangentMode=RCTM_User, time=0, value=0}
          {interpMode=RCIM_Cubic, tangentMode=RCTM_User, time=0.5, value=1}
          {tangentMode=RCTM_User, time=1, value=0}
         ]}}}
      Scale Curve=&Multiply_Float(A=0.4, B=$Particles.RegionMask.Float)
     )
     B=&FloatFromCurve(
      FloatCurve=DataInterface() {propertyValues {Curve {
         defaultValue=1
         keys [
          {interpMode=RCIM_Cubic, tangentMode=RCTM_User, time=0, value=4}
          {interpMode=RCIM_Cubic, tangentMode=RCTM_User, time=0.5, value=3}
          {tangentMode=RCTM_User, time=1, value=4}
         ]
        }}}
      CurveIndex=Linked() {linkedVariable {name=NPC.FortniteNPC.RemappedTimeOfDay_0-1, type=NiagaraFloat}}
     )
    )
    Scale RGB=&Multiply_VectorByFloat(
     Vector=&MakeVector(
      X=&FloatFromCurve(
       FloatCurve=DataInterface() {propertyValues {Curve {
          defaultValue=1
          keys [
           {interpMode=RCIM_Cubic, tangentMode=RCTM_User, time=0, value=1}
           <at 0.2464 0.45 RCIM_Cubic>
           {interpMode=RCIM_Cubic, tangentMode=RCTM_User, time=0.5, value=1}
           <at 0.7645 0.45 RCIM_Cubic>
           {tangentMode=RCTM_User, time=1, value=1}
          ]
         }}}
       CurveIndex=Linked() {linkedVariable {name=NPC.FortniteNPC.RemappedTimeOfDay_0-1, type=NiagaraFloat}}
      )
      Y=1
      Z=1
     )
     Float=&RandomRangeFloat(Minimum=1, Maximum=2)
    )
    ScaleA=true
    ScaleRGB=true
   ) {module=ScaleColor001}
   &CameraOffset(Camera Offset Amount=10) {module=CameraOffset001}
   &SubUVAnimation(
    "Sprite Renderer "=DataInterface() {propertyValues {SpriteRenderer=&self:Dirt_Blowing_0.NiagaraSpriteRendererProperties_1}}
    UseEndFrame=false
    UseStartFrame=false
   )
   &ScaleColor(
    Scale Alpha=&Lerp_Float(
     A=0
     B=0.9
     Alpha=&self:ScratchDynamicInput_05(
      End Distance:NiagaraFloat=1000
      Max Output:NiagaraFloat=1
      MCP Player Location:LinearColor=$NPC.FortniteNPC.PlayerPosition
      Minimum Output:NiagaraFloat=0
      Start Distance:NiagaraFloat=0
     )
    )
    ScaleA=true
    ScaleRGB=true
   )
  ]
  rendererClasses [&NiagaraSpriteRendererProperties]
  renderers [{rendererClass=&NiagaraSpriteRendererProperties}]
 }
 Skulls_Landscape001 {
  simTarget=GPUComputeSim
  emitterUpdateScript [
   &EmitterState()
   &SpawnRate(SpawnRate=&RandomRangeFloat(Minimum=15, Maximum=25), Use Spawn Probability=false)
  ]
  particleSpawnScript [
   &AddVelocity(Rotation Coordinate Space=::World, Velocity=<50 0 0>)
   &InitializeParticle(
    Color Maximum=&FNTimeOfDayColor(
     Night Color=<rgba 0.314 0.1418 0.1329 1>
     Morning Color=<rgba 0.6685 0.2254 0.1249 1>
     Day Color=<rgba 0.6685 0.2254 0.1249 1>
     Evening Color=<rgba 0.314 0.1418 0.1329 1>
    )
    Color Minimum=&FNTimeOfDayColor(
     Night Color=<rgba 0.314 0.1418 0.1329 1>
     Morning Color=<rgba 0.5506 0.1778 0.1574 1>
     Day Color=<rgba 0.5506 0.1778 0.1574 1>
     Evening Color=<rgba 0.314 0.1418 0.1329 1>
    )
    Color Mode=::NewEnumerator2
    Lifetime Mode=::NewEnumerator1
    Mesh Scale Mode=::NewEnumerator1
    Mesh Uniform Scale Max=10
    Mesh Uniform Scale Min=5
    Sprite Rotation Mode=::NewEnumerator2
    Sprite Size Mode=::NewEnumerator1
    Sprite UV Mode=::NewEnumerator1
    Uniform Sprite Size Max=1000
    Uniform Sprite Size Min=750
    UsePositionOffset=false
    Write Material Random=true
   )
   /SetVariables(Particles.Terrain Offset=20) {isSetParameters=true}
   &ShapeLocation(MaskSpawnGroup=false, Override Seed=false, Sphere Radius=2500, UseOwnerScale=true)
   &self:Sample_Landscape_02() {module=Sample_Landscape_02}
   &KillParticles() {module=KillParticles001}
   /Game/Effects/Niagara/Modules/Spawn/FN_Spawn_GBuffer_Color_WithinRangeV2(
    HueMax:NiagaraFloat=&OneMinusFloat(Float=17)
    HueMin:NiagaraFloat=&OneMinusFloat(Float=12)
   )
   /SetVariables(Particles.New Vector 3f:Vector3f=<0 0 1>) {isSetParameters=true}
   &DynamicMaterialParameters(
    Index 0 Param 1=&RandomRangeFloat(Minimum=50, Maximum=100)
    Index 0 Param 2=5
    Index 0 Param 3=15
    Index 0 Param 4=&RandomRangeFloat(Minimum=0, Maximum=0)
    Index 1 Param 1=50
    Index 2 Param 1=&RandomRangeFloat(Minimum=-0.1, Maximum=-0.1)
    Index 2 Param 2=&RandomRangeFloat(Minimum=0.1, Maximum=0)
    Index 2 Param 3=&RandomRangeFloat(Minimum=0, Maximum=0)
    Index 2 Param 4=&RandomRangeFloat(Minimum=0.2, Maximum=0.3)
    Module.Index0Param0Write=true
    Module.Index0Param1Write=true
    Module.Index0Param2Write=true
    Module.Index0Param3Write=true
    Module.Index1Param0Write=true
    Module.Index1Param1Write=true
    Module.Index1Param2Write=true
    Module.Index1Param3Write=true
    Module.Index2Param0Write=true
    Module.Index2Param1Write=true
    Module.Index2Param2Write=true
    Module.Index2Param3Write=true
    Write Parameter Index 1=true
    Write Parameter Index 2=true
   )
   /SetVariables(Particles.Linear Color=<rgba 1 0 0 1>, Particles.WaveScale:Vector3f=<6 6 6>) {isSetParameters=true}
   &self:Rejection_Sample_-_Normal_01(B=0.5, Landscape:Vector3f=$Particles.Landscape Normal, Spawn Vector:Vector3f=<0 0 1>) {module=Rejection_Sample_-_Normal_01}
   &ApplyInitialForces()
   &Color() {enabled=false}
   &CameraOffset(Camera Offset Amount=10) {enabled=false}
   &KillParticlesInVolume(
    Box Size:Vector3f=<6000 6000 6000>
    Kill Shape=::NewEnumerator1
    Volume Origin:Vector3f=&TransformPosition(Position=<-30164 12004 1932>, Source Space=::Simulation, Destination Space=::World)
   ) {module=KillParticlesInVolume006}
   &KillParticlesInVolume(
    Box Size:Vector3f=<6000 6000 6000>
    Kill Shape=::NewEnumerator1
    Volume Origin:Vector3f=&TransformPosition(Position=<-76592 7456 932>, Source Space=::Simulation, Destination Space=::World)
   ) {module=KillParticlesInVolume007}
   &KillParticlesInVolume(
    Box Size:Vector3f=<6000 6000 6000>
    Kill Shape=::NewEnumerator1
    Volume Origin:Vector3f=&TransformPosition(Position=<-69468 2040 932>, Source Space=::Simulation, Destination Space=::World)
   ) {module=KillParticlesInVolume008}
   &KillParticlesInVolume(
    Box Size:Vector3f=<3000 3750 1500>
    Kill Shape=::NewEnumerator1
    Volume Origin:Vector3f=&TransformPosition(Position=<-98172 -12784 1536>, Source Space=::Simulation, Destination Space=::World)
   ) {module=KillParticlesInVolume009}
   &KillParticlesInVolume(
    Box Size:Vector3f=<1750 5250 1500>
    Kill Shape=::NewEnumerator1
    Volume Origin:Vector3f=&TransformPosition(Position=<-99444 -46664 976>, Source Space=::Simulation, Destination Space=::World)
   ) {module=KillParticlesInVolume010}
   &KillParticlesInVolume(
    Box Size:Vector3f=<4000 2000 1500>
    Kill Shape=::NewEnumerator1
    Volume Origin:Vector3f=&TransformPosition(Position=<-98812 -48124 976>, Source Space=::World, Destination Space=::Local)
   ) {module=KillParticlesInVolume011}
  ]
  particleUpdateScript [
   &self:Sample_Landscape_UPdate_03(Bias:NiagaraFloat=$Particles.Terrain Offset) {module=Sample_Landscape_UPdate_03}
   &ParticleState()
   &SpriteRotationRate(Rotation Rate=-60)
   &ScaleSpriteSize(Uniform Curve Sprite Scale=DataInterface() {propertyValues {Curve {keys [{interpMode=RCIM_Cubic, leaveTangent=3.9356, tangentMode=RCTM_Break, time=0, value=0}, <at 1 1 RCIM_Cubic>]}}})
   &WindForce(
    Friction=0.25
    Friction Distance Limit=5
    Normal Friction Mask=0.5
    Offset By Time:Vector3f=<0 0 0>
    Wind Speed=<750 0 -100>
    Wind Speed Scale=&Lerp_Float(A=0.25, B=0.5, Alpha=$NPC.FortniteNPC.WindGustsSystem_Active)
   )
   &Drag(
    Drag=&FloatFromCurve(
     FloatCurve=DataInterface() {propertyValues {Curve {keys [{interpMode=RCIM_Cubic, tangentMode=RCTM_Break, time=0, value=0.25}, {tangentMode=RCTM_User, time=1, value=1}]}}}
     Scale Curve=8
    )
    Use Linear Drag=true
    Use Rotational Drag=true
   )
   /Niagara/Modules/Math/AvoidDistanceFieldSurfaces_GPU(
    Avoid Nearest Surface=true
    Avoid Oncoming Surfaces=true
    Falloff Distance=1
    Max Trace Distance=1
    Nearest Surface Avoidance Strength=1
    Obstactle Avoidance Strength=50
    Trace Vector:Vector3f=&Multiply_Vector(A=$Particles.Velocity, B=<1 1 0>)
   ) {enabled=false}
   &SolveForcesAndVelocity(Clamp Velocity=false, Limit Acceleration=false)
   &ScaleSpriteSizeBySpeed(Max Scale Factor=<2 8>, Min Scale Factor=<0.05 0.05>)
   &RegionMask(
    Invert Region=true
    Region Dimensions:Vector3f=<600 600 600>
    Region Falloff=1
    Region Origin:Vector3f=&MakeVectorFromLinearColorRGB(Color=$NPC.FortniteNPC.PlayerPosition)
    Region Type:ENiagaraDebugDrawShape=::NewEnumerator1
    UseOffset=false
    UseRotation=false
   )
   &RegionMask(
    Region Dimensions:Vector3f=<6000 6000 6000>
    Region Origin:Vector3f=&TransformPosition(Position=<-30164 12004 1932>, Source Space=::Simulation, Destination Space=::World)
    Region Type:ENiagaraDebugDrawShape=::NewEnumerator1
    UseOffset=false
    UseRotation=false
   ) {enabled=false, module=RegionMask001}
   &ScaleColor(
    Scale Alpha=&FloatFromCurve(
     FloatCurve=DataInterface() {propertyValues {Curve {keys [
         {interpMode=RCIM_Cubic, tangentMode=RCTM_User, time=0, value=0}
         {interpMode=RCIM_Cubic, tangentMode=RCTM_User, time=0.5, value=1}
         {tangentMode=RCTM_User, time=1, value=0}
        ]}}}
     Scale Curve=&Multiply_Float(A=0.5, B=$Output.RegionMask.RegionMask)
    )
    Scale RGB=&Multiply_VectorByFloat(Vector=<1 1 1>, Float=&RandomRangeFloat(Minimum=1, Maximum=2))
    ScaleA=true
    ScaleRGB=true
   ) {module=ScaleColor001}
   /SetVariables(
    Particles.Linear Color=&MultiplyLinearColor(
     Linear Color=$Particles.Color
     Scale Factor=&MakeVector4_FromFloat(Float=&FloatFromCurve(
       FloatCurve=DataInterface() {propertyValues {Curve {keys [
           {interpMode=RCIM_Cubic, tangentMode=RCTM_User, time=0, value=0}
           {interpMode=RCIM_Cubic, tangentMode=RCTM_User, time=0.5, value=1}
           {tangentMode=RCTM_User, time=1, value=0}
          ]}}}
       Scale Curve=$System.ColorIntensityMult
      ))
    )
    Particles.Linear Color001:LinearColor=&MultiplyLinearColor(
     Linear Color=&ColorFromCurve(ColorCurve=DataInterface() {propertyValues {
        AlphaCurve {keys [<at 0 0 RCIM_Cubic>,<at 0.1218 1>,<at 1 0 RCIM_Cubic>]}
        BlueCurve {keys [<at 0 0.0688 RCIM_Cubic>,<at 1 1 RCIM_Cubic>]}
        GreenCurve {keys [<at 0 1 RCIM_Cubic>,<at 1 1 RCIM_Cubic>]}
        RedCurve {keys [<at 0 0 RCIM_Cubic>,<at 1 1 RCIM_Cubic>]}
       }})
     Scale Factor=&MakeVector4_FromFloat(Float=&FloatFromCurve(
       FloatCurve=DataInterface() {propertyValues {Curve {keys [<at 0 1>,<at 1 0>]}}}
       Scale Curve=&Multiply_Float(A=$System.ColorIntensityMult, B=4)
      ))
    )
   ) {isSetParameters=true}
   &CameraOffset(Camera Offset Amount=10) {module=CameraOffset001}
   &SubUVAnimation(
    Play Rate=0.15
    "Sprite Renderer "=DataInterface() {propertyValues {SpriteRenderer=&self:Skulls_Landscape001_1.NiagaraSpriteRendererProperties_1}}
    SubUV Animation Mode:ENiagara_SubUVLookupModeV2=::NewEnumerator3
    UseEndFrame=false
    UsePlayRateScale=false
    UseStartFrame=false
   )
   &ScaleSpriteSize(Uniform Curve Index=&self:ScratchDynamicInput_02(
     End Distance:NiagaraFloat=3500
     Max Output:NiagaraFloat=1
     MCP Player Location:LinearColor=$NPC.FortniteNPC.PlayerPosition
     Minimum Output:NiagaraFloat=0
     Start Distance:NiagaraFloat=0
    )) {module=ScaleSpriteSize001}
   &KillParticlesInVolume(
    Box Size:Vector3f=<6000 6000 6000>
    Kill Shape=::NewEnumerator1
    Volume Origin:Vector3f=&TransformPosition(Position=<-30164 12004 1932>, Source Space=::Simulation, Destination Space=::World)
   )
   &KillParticlesInVolume(
    Box Size:Vector3f=<6000 6000 6000>
    Kill Shape=::NewEnumerator1
    Volume Origin:Vector3f=&TransformPosition(Position=<-76592 7456 932>, Source Space=::Simulation, Destination Space=::World)
   ) {module=KillParticlesInVolume001}
   &KillParticlesInVolume(
    Box Size:Vector3f=<6000 6000 6000>
    Kill Shape=::NewEnumerator1
    Volume Origin:Vector3f=&TransformPosition(Position=<-69468 2040 932>, Source Space=::Simulation, Destination Space=::World)
   ) {module=KillParticlesInVolume002}
   &KillParticlesInVolume(
    Box Size:Vector3f=<3000 3750 1500>
    Kill Shape=::NewEnumerator1
    Volume Origin:Vector3f=&TransformPosition(Position=<-98172 -12784 1536>, Source Space=::Simulation, Destination Space=::World)
   ) {module=KillParticlesInVolume003}
   &KillParticlesInVolume(
    Box Size:Vector3f=<1750 5250 1500>
    Kill Shape=::NewEnumerator1
    Volume Origin:Vector3f=&TransformPosition(Position=<-99444 -46664 976>, Source Space=::Simulation, Destination Space=::World)
   ) {module=KillParticlesInVolume004}
   &KillParticlesInVolume(
    Box Size:Vector3f=<4000 2000 1500>
    Kill Shape=::NewEnumerator1
    Volume Origin:Vector3f=&TransformPosition(Position=<-98812 -48124 976>, Source Space=::World, Destination Space=::Local)
   ) {module=KillParticlesInVolume005}
  ]
  rendererClasses [&NiagaraSpriteRendererProperties]
  renderers [{rendererClass=&NiagaraSpriteRendererProperties}]
 }
 Tumbleweed_Big {
  simTarget=GPUComputeSim
  emitterUpdateScript [
   &EmitterState()
   &SpawnBurst_Instantaneous(
    Spawn Count=&RandomRangeInt(Minimum=1, Maximum=1)
    Use Spawn Probability=false
    UseLoopCountLimit=false
   ) {enabled=false}
   &SpawnRate(SpawnRate=1.2, Use Spawn Probability=false)
  ]
  particleSpawnScript [
   &InitializeParticle(
    Color=&FN_Select_Color_from_Time_Of_Day(
     Morning=&SelectLinearColorFromArray(
      Array Sampling Mode=::NewEnumerator0
      Direct Array Index=&RandomRangeInt(Minimum=0, Maximum=2)
      Color Selection Array=DataInterface() {propertyValues {ColorData [<rgba 0.2063 0.0723 0.0193 1>, <rgba 0.3478 0.1459 0.0661 1>, <rgba 0.31 0.1929 0.0741 1>]}}
      Random Seed=0
     )
     Day=&SelectLinearColorFromArray(
      Array Sampling Mode=::NewEnumerator0
      Direct Array Index=&RandomRangeInt(Minimum=0, Maximum=2)
      Color Selection Array=DataInterface() {propertyValues {ColorData [<rgba 0.2063 0.0723 0.0193 1>, <rgba 0.3478 0.1459 0.0661 1>, <rgba 0.31 0.1929 0.0741 1>]}}
      Random Seed=0
     )
     Evening=&SelectLinearColorFromArray(
      Array Sampling Mode=::NewEnumerator0
      Direct Array Index=&RandomRangeInt(Minimum=0, Maximum=2)
      Color Selection Array=DataInterface() {propertyValues {ColorData [<rgba 0.2063 0.0723 0.0193 1>, <rgba 0.3478 0.1459 0.0661 1>, <rgba 0.31 0.1929 0.0741 1>]}}
      Random Seed=0
     )
     Night=&SelectLinearColorFromArray(
      Array Sampling Mode=::NewEnumerator0
      Direct Array Index=&RandomRangeInt(Minimum=0, Maximum=2)
      Color Selection Array=DataInterface() {propertyValues {ColorData [<rgba 0.2063 0.0723 0.0193 1>, <rgba 0.3478 0.1459 0.0661 1>, <rgba 0.31 0.1929 0.0741 1>]}}
      Random Seed=0
     )
    )
    Lifetime=&RandomRangeFloat(Minimum=3, Maximum=6)
    Mesh Scale Mode=::NewEnumerator1
    Position Offset=<0 0 15>
    UsePositionOffset=true
    Write Material Random=true
   )
   &ShapeLocation(MaskSpawnGroup=false, Non Uniform Scale=<1 1 0.01>, Override Seed=false, Sphere Radius=3000, UseOwnerScale=true)
   &SampleLandscape(Height Offset=5)
   &self:Sample_Landscape_02() {enabled=false, module=Sample_Landscape_02}
   &AddVelocity(
    Rotation Coordinate Space=::World
    Velocity=&RandomRangeVector(Minimum=<300 -300 0>, Maximum=<500 500 0>)
   ) {enabled=false, module=AddVelocity001}
   &KillParticlesInVolume(
    Sphere Radius=1000
    Volume Origin:Vector3f=&MakeVectorFromLinearColorRGB(Color=$NPC.FortniteNPC.PlayerPosition)
   ) {enabled=false}
   &self:Rejection_Sample_-_Normal_01(B=0.5, Landscape:Vector3f=$Particles.Landscape Normal, Spawn Vector:Vector3f=<0 0 1>) {enabled=false, module=Rejection_Sample_-_Normal_01}
  ]
  particleUpdateScript [
   &ParticleState()
   &RegionMask(
    Offset:Vector3f=&MakeVectorFromLinearColorRGB(Color=$NPC.FortniteNPC.PlayerPosition)
    Offset Space:ENiagaraCoordinateSpace=::Local
    Region Radius=1500
    UseOffset=true
    UseRotation=false
   )
   &ScaleMeshSize(Scale Factor=&Multiply_VectorByFloat(
     Vector=&RandomRangeVector(Minimum=<0.5 0.5 0.5>, Maximum=<1 1 1>)
     Float=&FloatFromCurve(
      FloatCurve=DataInterface() {propertyValues {Curve {
         defaultValue=0
         keys [
          {arriveTangent=6.757, interpMode=RCIM_Cubic, leaveTangent=6.757, tangentMode=RCTM_User, time=0, value=0}
          {interpMode=RCIM_Cubic, tangentMode=RCTM_User, time=0.1066, value=1}
          {arriveTangent=-0.1331, interpMode=RCIM_Cubic, leaveTangent=-0.1331, tangentMode=RCTM_User, time=0.5008, value=0.9748}
          {arriveTangent=-1.1973, tangentMode=RCTM_User, time=1, value=0}
         ]
        }}}
      Scale Curve=&Lerp_Float(A=1, B=0, Alpha=$Output.RegionMask.RegionMask)
     )
    ))
   &Collision(
    Age Colliding Particles=false
    Friction=1
    Friction During a Bounce=0.5
    GPU Collision Type=::NewEnumerator2
    KillOnCollision=false
    Mesh Dimensions:Vector3f=<62 62 62>
    Module.Control Rest State Via Surface Normal=false
    Module.EnableMaxCollisionCount=false
    Module.Randomize Collision Normal Vector=false
    Radius Calculation Type=::NewEnumerator1
    Restitution=&RandomRangeFloat(Minimum=0, Maximum=0.3)
   ) {module=Collision001}
   &WindForce(
    Friction=0.25
    Friction Distance Limit=5
    Normal Friction Mask=0.5
    Offset By Time:Vector3f=<0 0 0>
    Wind Speed=<1000 0 -15>
    Wind Speed Scale=&Lerp_Float(A=0.4, B=0.7, Alpha=$NPC.FortniteNPC.WindGustsSystem_Active)
   )
   &AerodynamicDrag(
    Aerodynamic Drag:NiagaraFloat=&RandomRangeFloat(Minimum=2, Maximum=5)
    Aerodynamic Rotational Drag=10
    Lift Contribution:NiagaraFloat=&RandomRangeFloat(Minimum=0, Maximum=0.1)
    Use Linear Drag=true
    Use Rotational Drag=true
   )
   &UpdateMeshOrientation(
    Rotation Rate=&RandomRangeFloat(Minimum=2, Maximum=5)
    Rotation Vector=&RandomRangeVector(Minimum=<0.25 0 0>, Maximum=<0.65 0 0>)
   )
   &DynamicMaterialParameters(
    Index 0 Param 1=0.7
    Index 0 Param 2=&FloatFromCurve(FloatCurve=DataInterface() {propertyValues {Curve {keys [
         {interpMode=RCIM_Cubic, tangentMode=RCTM_User, time=0, value=0.75}
         {interpMode=RCIM_Cubic, tangentMode=RCTM_User, time=0.5, value=0}
         {tangentMode=RCTM_User, time=1, value=0.75}
        ]}}})
    Index 0 Param 3=&RandomRangeFloat(Minimum=0, Maximum=1)
    Module.Index0Float0Write=true
    Module.Index0Float1Write=true
    Module.Index0Float2Write=true
    Module.Index0Float3Write=true
   )
   &Drag(Drag=4, Use Linear Drag=true, Use Rotational Drag=true) {enabled=false, module=Drag001}
   &GravityForce(Gravity=<0 0 -5000>)
   &SolveForcesAndVelocity(Clamp Velocity=false, Limit Acceleration=false)
   &AlignSpriteToMeshOrientation(SetAlignment=true, SetFacing=true)
   &ScaleColor(
    Scale Alpha=&Lerp_Float(
     A=0
     B=1
     Alpha=&self:ScratchDynamicInput_01(
      End Distance:NiagaraFloat=850
      Max Output:NiagaraFloat=1
      MCP Player Location:LinearColor=$NPC.FortniteNPC.PlayerPosition
      Minimum Output:NiagaraFloat=0
      Start Distance:NiagaraFloat=0
     )
    )
    ScaleA=true
    ScaleRGB=true
   )
  ]
  rendererClasses [&NiagaraMeshRendererProperties]
  renderers [{rendererClass=&NiagaraMeshRendererProperties, rendererIndex=0}]
 }
}
summary {
 systemName=NS_Env_DesertBiome_BlowingSand
 emitters [
  {bEnabled=false, emitterName=Dirt_Blowing, simTarget=GPUComputeSim, rendererClasses [&NiagaraSpriteRendererProperties]}
  {emitterName=Tumbleweed_Big, simTarget=GPUComputeSim, rendererClasses [&NiagaraMeshRendererProperties]}
  {emitterName=Skulls_Landscape001, simTarget=GPUComputeSim, rendererClasses [&NiagaraSpriteRendererProperties]}
 ]
 userVariables [
  User.Landscape {
   type=NiagaraDataInterfaceLandscape
   description=""
   defaultValue {
    struct=NiagaraExt_VariableValue_DataInterface
    value {
     dataInterface=&self:NiagaraDataInterfaceLandscape_0
     dataInterfaceClass=&NiagaraDataInterfaceLandscape
    }
   }
  }
  User.ParticleScale {type=Vector3f, description="", defaultValue {struct=Vector3f, value=<0.1 0.1 0.1>}}
 ]
}
